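; Parent gameplay ability that rescales a Fortnite player pawn's max health and shield while active, gated by an enabled scalable-float row.
; It saves health and shield as percentages, applies a SetByCaller GameplayEffect with row-driven maxima, then restores the percentages.
; A DBNO tag watcher sets health from the saved percentage when downed, scaled by a revive percent when the tag clears.
; Ending the ability removes the effect, preserving current percentages.
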(event Ability|EventActivateAbility
  (bind _playerpawn (Variables|ParentHealthShield|GetPlayerPawn))
  (bind _returnvalue (FortPawn|GetMaxHealth _playerpawn))
  (bind _healthpercent (Variables|ParentHealthShield|GetHealthPercent))
  (bind _returnvalue_2 (* _returnvalue _healthpercent))
  ;[1 1 1] Begin Play - Save Out Pawn, Check If enabled
  (if (CallFunction|RowtoBool)
    (bind _asfort_player_pawn (Utilities|Casting|CastToFortPlayerPawn (Ability|GetAvatarActorfromActorInfo)))
    (Variables|ParentHealthShield|SetPlayerPawn _asfort_player_pawn)
    ;[1 1 1] Save Health and Shield Values then apply GE
    (CallFunction|SaveHealthandShieldPercents)
    (bind _returnvalue_3 (CallFunction|SetHealthandShieldGE))
    (bind _returnvalue_4 (Ability|ApplyGameplayEffectSpecToOwner _returnvalue_3))
    (Variables|ParentHealthShield|SetGEH_MaxHealthShield _returnvalue_4)
    (CallFunction|SetHealthandShieldPercents)
    ;[1 1 1] Handle DBNO
    (Ability|Tasks|WaitGameplayTagAdd "(TagName=\"Gameplay.Status.DBNO\")"
      (:Added
        (Ability|Tasks|WaitGameplayTagRemove "(TagName=\"Gameplay.Status.DBNO\")"
          (:then
            (FortPawn|SetHealth _playerpawn _returnvalue_2))
          (:Removed
            (FortPawn|SetHealth _playerpawn (* _returnvalue_2 (Variables|ParentHealthShield|GetDBNORevivePercent)))))))))

(event Ability|EventOnEndAbility
  (params (bWasCancelled :type Boolean))
  ;[1 1 1] On End Ability - Reset Health and Shield and Remove GE
  (CallFunction|SaveHealthandShieldPercents)
  (Ability|RemoveGameplayEffectFromOwnerWithHandle (Variables|ParentHealthShield|GetGEH_MaxHealthShield))
  (CallFunction|SetHealthandShieldPercents))

(fn RowToBool
  (return (Math|Conversions|ToBoolean(Integer) (Math|Float|Ceil (ScalableFloat|GetValueatLevel (Variables|ParentHealthShield|GetRow_Enabled))))))

(fn SaveHealthAndShieldPercents
  (bind _playerpawn (Variables|ParentHealthShield|GetPlayerPawn))
  (Utilities|IsValid _playerpawn
    (:Is%20Valid
      (Variables|ParentHealthShield|SetHealthPercent (/ (FortPawn|GetHealth _playerpawn) (FortPawn|GetMaxHealth _playerpawn)))
      (Variables|ParentHealthShield|SetShieldPercent (/ (FortPawn|GetShield _playerpawn) (FortPawn|GetMaxShield _playerpawn))))))

(fn SetHealthAndShieldGE
  (bind _returnvalue (Ability|GameplayEffect|AssignTagSetbyCallerMagnitude (Ability|MakeOutgoingGameplayEffectSpec (Variables|ParentHealthShield|GetGE_MaxHealthShield) 0.0) "(TagName=\"SetByCaller.MaxHealth\")" (ScalableFloat|GetValueatLevel (Variables|ParentHealthShield|GetRow_HealthMax))))
  (bind _returnvalue_2 (Ability|GameplayEffect|AssignTagSetbyCallerMagnitude _returnvalue "(TagName=\"SetByCaller.MaxShield\")" (ScalableFloat|GetValueatLevel (Variables|ParentHealthShield|GetRow_ShieldMax))))
  (return _returnvalue_2))

(fn SetHealthAndShieldPercents
  (bind _playerpawn (Variables|ParentHealthShield|GetPlayerPawn))
  (Utilities|IsValid _playerpawn
    (:Is%20Valid
      (FortPawn|SetShield _playerpawn (* (Variables|ParentHealthShield|GetShieldPercent) (FortPawn|GetMaxShield _playerpawn)))
      (FortPawn|SetHealth _playerpawn (* (FortPawn|GetMaxHealth _playerpawn) (Variables|ParentHealthShield|GetHealthPercent))))))
